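; GA_LemonExam_Heavy is a gameplay ability, apparently the Heavy ability, that gives audio feedback when activation fails on cooldown.
; Its OverrideFailedReason override checks the failed-reason tags for Activate.Failed.Cooldown and plays the ability-on-cooldown sound as 2D audio.
; A Sequence then passes the tags to the parent implementation and returns that result.
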(fn OverrideFailedReason
  (params (FailedReason :type Gameplay%20Tag%20Container%20Structure%20%28by%20ref%29))
  (Utilities|FlowControl|Sequence
    (:then_0
      (if (GameplayTags|HasTag FailedReason "(TagName=\"Activate.Failed.Cooldown\")")
        (Audio|PlaySound2D (Variables|Default|GetAbilityOnCooldownSound))))
    (:then_1
      (bind _overridenfailedreason (|CallParentFunction FailedReason))
      (return _overridenfailedreason))))
